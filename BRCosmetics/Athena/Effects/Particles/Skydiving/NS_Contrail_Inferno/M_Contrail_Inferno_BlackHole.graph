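# Translucent unlit Niagara sprite and mesh material for the swirling black hole in the Inferno skydiving contrail.
# Emissive layers radially mapped smoke and space textures, noise-warped smoke twisted by sphere-mask-offset rotation, and triplanar camera-vector stars, tinted by particle color and time of day.
# Opacity combines the swirl, a hollow center, a noise-distorted rim and view-angle fade, halved on Low/Medium quality.
# Dynamic parameters drive hole size and rotation speed.
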Material /BRCosmetics/Athena/Effects/Particles/Skydiving/NS_Contrail_Inferno/M_Contrail_Inferno_BlackHole
graph {
 properties {
  bUsedWithNiagaraMeshParticles=true
  bUsedWithNiagaraSprites=true
  blendMode=BLEND_Translucent
  pixelDepthOffsetMode=PDOM_Legacy
  shadingModel=MSM_Unlit
  twoSided=true
  usedShadingModels=Unlit
 }
 nodes [
  n1 = ParticleColor()
  n2 = Multiply(
   A=MaterialFunctionCall() {materialFunction=&VectorToRadialValue}.Radial Coordinates
   B=Constant2Vector() {g=1, r=1}
  )
  n3 = Multiply(
   A=Multiply(
    A=ScalarParameter() {defaultValue=1, parameterName=Rotator speed}
    B=DynamicParameter() {paramNames [BlackHole size, Rotator Speed, Param3, Param4]}.Rotator Speed
   )
   B=Time()
  )
  n4 = MaterialFunctionCall() {materialFunction=&VectorToRadialValue}
  n5 = Constant2Vector() {g=2, r=2}
  n6 = TextureCoordinate()
  n7 = TextureSample!(
   UVs=Panner(Coordinate=Multiply(A=@n4.Radial Coordinates, B=@n5), Time=@n3) {speedX=0.03, speedY=0.1}
   Tex
   "DDX(UVs)"=DDX(Value=@n6)
   "DDY(UVs)"=DDY(Value=@n6)
   Apply View MipBias
  ) {mipValueMode=TMVM_Derivative, texture=&T_Inky_Smoke_Tile}
  n8 = TextureCoordinate()
  n9 = TextureSample(
   UVs=Add(A=Panner(Coordinate=@n2, Time=@n3) {speedX=0.1}, B=Multiply(A=@n7.R) {constB=0.2})
   "DDX(UVs)"=DDX(Value=@n8)
   "DDY(UVs)"=DDY(Value=@n8)
  ) {mipValueMode=TMVM_Derivative, texture=/Game/Gadgets/Assets/VinderTech_GliderChute/Glider_Cube_Queen_Female/FX/T_CubeQueen_Glider_Stair}
  n10 = Multiply(A=@n9.RGB) {constB=7}
  n11 = Add(
   A=MaterialFunctionCall() {materialFunction=&VectorToRadialValue}.Radial Coordinates
   B=Constant2Vector() {g=-0.05, r=-0.5}
  )
  n12 = TextureCoordinate()
  n13 = DDX(Value=@n12)
  n14 = DDY(Value=@n12)
  n15 = TextureSample(
   UVs=Add(
    A=MaterialFunctionCall() {materialFunction=&VectorToRadialValue}.Radial Coordinates
    B=Constant2Vector() {g=0.11}
   )
   "DDX(UVs)"=@n13
   "DDY(UVs)"=@n14
  ) {mipValueMode=TMVM_Derivative, texture=/Game/Gadgets/Assets/VinderTech_GliderChute/Glider_Cube_Queen_Female/FX/T_Cheap_blue_space_color}
  n16 = Add(
   A=Multiply(
    A=Constant3Vector() {constant=<rgba 0.2284 0.0012 1 0>}
    B=TextureSample(UVs=@n11, "DDX(UVs)"=@n13, "DDY(UVs)"=@n14) {mipValueMode=TMVM_Derivative, texture=&T_Inky_Smoke_Tile}.R
   )
   B=Multiply(A=@n15.RGB, B=Constant3Vector() {constant=<rgba 1.2685 0.2555 3 0>})
  ) {constB=0}
  n17 = Subtract(A=TextureSample!(UVs=TextureCoordinate() {uTiling=0.7, vTiling=0.7}, Tex, Apply View MipBias) {samplerType=SAMPLERTYPE_LinearColor, texture=&T_NoisePack_LinearColor}.RGB) {constB=0.5}
  n18 = Add(A=TextureCoordinate(), B=Multiply(A=ComponentMask(None=@n17) {g=true, r=true}) {constB=0.1})
  n19 = Multiply(
   A=ScalarParameter() {defaultValue=1, parameterName=Rotator speed}
   B=DynamicParameter() {paramNames [BlackHole size, Rotator Speed, Param3, Param4]}.Rotator Speed
  )
  n20 = SphereMask(A=TextureCoordinate(), B=Constant() {r=0.5}, Radius=Constant() {r=1}, Hardness=Constant()) {hardnessPercent=0}
  n21 = Multiply(A=@n20, B=@n20)
  n22 = Multiply(A=@n19, B=Add(A=Time(), B=Multiply(A=@n21) {constB=10})) {constA=1}
  n23 = TextureCoordinate()
  n24 = SphereMask(A=Add(B=TextureCoordinate()), B=Constant() {r=0.5}) {attenuationRadius=1, hardnessPercent=0}
  n25 = Power(
   Base=TextureSample!(UVs=Rotator(Coordinate=@n18, Time=@n22), Tex, "DDX(UVs)"=DDX(Value=@n23), "DDY(UVs)"=DDY(Value=@n23), Apply View MipBias) {mipValueMode=TMVM_Derivative, texture=&T_Inky_Smoke_Tile_Chromatic}.RGB
   Exp=LinearInterpolate(Alpha=OneMinus(None=@n24))
  )
  n26 = ScalarParameter() {defaultValue=0.7, parameterName=Main Bright}
  n27 = TextureObjectParameter() {
   group=Galaxy
   parameterName=Stars
   samplerType=SAMPLERTYPE_Masks
   texture=/Game/Characters/Player/Male/Medium/Bodies/M_Med_Soldier_04/Skins/Galaxy/Textures/Celestial_T_Stars
  }
  n28 = ScalarParameter() {defaultValue=3, group=Galaxy, parameterName=Bright Star Tiling}
  n29 = ComponentMask(None=MaterialFunctionCall!(CameraVector, Texture=@n27, Tiling=Multiply(A=@n28), Offset, AxisFadeContrast) {materialFunction=/Engine/Functions/Engine_MaterialFunctions01/Texturing/TriplanarCameraVector}.Result) {r=true}
  n30 = TextureSample!(UVs=Panner(Coordinate=TextureCoordinate() {uTiling=2, vTiling=2}) {speedX=0.133, speedY=0.2}, Tex, Apply View MipBias) {samplerType=SAMPLERTYPE_Masks, texture=&Celestial_T_Stars}
  n31 = Power(Base=@n30.G) {constExponent=1.5}
  n32 = Time()
  n33 = TextureObject() {samplerType=SAMPLERTYPE_Masks, texture=&Celestial_T_Stars}
  n34 = ComponentMask(None=MaterialFunctionCall!(
    CameraVector
    Rotation Speed=Multiply(A=@n32) {constB=-0.125}
    Texture=@n33
    Tiling=Constant3Vector() {constant=<rgba 7 7 7 0>}
    Offset
    AxisFadeContrast
   ) {materialFunction=/Game/Weapons/WeaponSkins/MaterialFunction/TriplanarCameraVector_Rotated}.Result) {b=true}
  n35 = Add(A=Multiply(A=Multiply(A=@n29, B=@n29), B=@n31), B=Multiply(A=Multiply(A=@n34, B=@n34), B=@n31))
  n36 = Add(A=Add(A=Multiply(A=@n35) {constB=10}) {constB=0}, B=Add(A=Add(A=@n10, B=@n16), B=Multiply(A=@n25, B=@n26) {constB=0.7}) {constB=0})
  n37 = Constant() {r=1}
  n38 = VectorParameter() {parameterName=Color Night, defaultValue=<rgba 2 0.6849 1.5726 0>}
  MaterialFunctionCall(Emissive=Multiply(
    A=Multiply(A=@n1.RGB, B=@n36)
    B=MaterialFunctionCall(Color_Morning=@n37, Color_Day=@n37, Color_Evening=@n37, Color_Night=@n38.RGB) {materialFunction=&Color_Time_of_Day}.Result
   )) {materialFunction=&MF_FX_EmissiveAdjustments}
  n39 = SphereMask(A=TextureCoordinate(), B=Constant() {r=0.5}, Radius=Constant() {r=1}, Hardness=Constant())
  n40 = Add(A=Multiply(A=Constant() {r=1}) {constB=0.25}, B=Multiply(A=@n39, B=@n39))
  n41 = Cosine(None=@n40)
  n42 = Subtract(A=Multiply(A=TextureCoordinate())) {constB=0.5}
  n43 = ComponentMask(None=@n42) {r=true}
  n44 = Sine(None=@n40)
  n45 = ComponentMask(None=@n42) {g=true}
  n46 = AppendVector(A=Subtract(A=Multiply(A=@n41, B=@n43), B=Multiply(A=@n44, B=@n45)), B=Add(A=Multiply(A=@n43, B=@n44), B=Multiply(A=@n45, B=@n41)))
  Panner(Coordinate=@n46) {speedX=0.1, speedY=0.1}
  n47 = MaterialFunctionCall("FadeSensitivity (optional)"=Constant() {r=1}) {materialFunction=&FadeBasedOnViewAngle}
  n48 = TextureCoordinate()
  n49 = TextureSample(UVs=TextureCoordinate() {uTiling=0.7, vTiling=0.7}) {samplerType=SAMPLERTYPE_LinearColor, texture=&T_NoisePack_LinearColor}
  n50 = Multiply(A=ComponentMask(None=Subtract(A=@n49.RGB) {constB=0.5}) {g=true, r=true}) {constB=0.1}
  n51 = SphereMask(A=TextureCoordinate(), B=Constant() {r=0.5}, Radius=Constant() {r=1}, Hardness=Constant()) {hardnessPercent=0}
  n52 = Add(A=Time(), B=Multiply(A=Multiply(A=@n51, B=@n51)) {constB=10})
  n53 = TextureCoordinate()
  n54 = TextureSample!(
   UVs=Rotator(Coordinate=Add(A=@n48, B=@n50), Time=Multiply(A=@n19, B=@n52) {constA=1})
   Tex
   "DDX(UVs)"=DDX(Value=@n53)
   "DDY(UVs)"=DDY(Value=@n53)
   Apply View MipBias
  ) {mipValueMode=TMVM_Derivative, texture=&T_Cheap_blue_space_color}
  n55 = Add(B=TextureCoordinate())
  n56 = Constant() {r=0.5}
  n57 = LinearInterpolate(Alpha=OneMinus(None=SphereMask(A=@n55, B=@n56) {attenuationRadius=1, hardnessPercent=0}))
  n58 = Add(B=TextureCoordinate())
  n59 = Constant() {r=0.5}
  n60 = Multiply(
   A=ScalarParameter() {defaultValue=0.5, parameterName=Radius}
   B=DynamicParameter() {paramNames [BlackHole size, Rotator Speed, Param3, Param4]}.BlackHole size
  )
  n61 = Multiply(A=Power(Base=SphereMask(A=@n58, B=@n59, Radius=@n60) {attenuationRadius=0.5, hardnessPercent=0})) {constB=2}
  n62 = Multiply(A=Add(A=Power(Base=@n54.RGB, Exp=@n57)) {constB=0}, B=OneMinus(None=Saturate(None=@n61)))
  n63 = TextureCoordinate()
  n64 = Multiply(
   A=ScalarParameter() {defaultValue=1, parameterName=Rotator speed}
   B=DynamicParameter() {paramNames [BlackHole size, Rotator Speed, Param3, Param4]}.Rotator Speed
  )
  n65 = Time()
  n66 = Multiply(
   A=Rotator(Coordinate=@n46, Time=Multiply(A=@n64, B=@n65)) {speed=0.3}
   B=ComponentMask(None=VectorParameter() {parameterName=UV, defaultValue=<rgba 1 1 1 0>}.RGB) {g=true, r=true}
  )
  n67 = MaterialFunctionCall(Float3=Multiply(A=TextureSample!(UVs=@n66, Tex, Apply View MipBias) {samplerType=SAMPLERTYPE_LinearColor, texture=&T_NoisePack_LinearColor}.RGB) {constB=0.1}) {materialFunction=&BreakOutFloat3Components}
  n68 = SphereMask(A=Subtract(A=Add(A=@n63, B=@n67.R), B=@n67.G), B=Constant() {r=0.5}, Radius=Constant() {r=0.45}, Hardness=Constant())
  n69 = Multiply(A=Multiply(A=@n1.A, B=Multiply(A=@n62, B=@n68)), B=Saturate(None=Multiply(A=@n47.Result) {constB=10}))
  n70 = Multiply(A=@n69) {constB=0.5}
  QualitySwitch(Default=@n69, Low=@n70, High=@n69, Medium=@n70, Epic=@n69)
  TextureCoordinate()
  TextureCoordinate() {uTiling=0.3, vTiling=0.3}
 ]
}
