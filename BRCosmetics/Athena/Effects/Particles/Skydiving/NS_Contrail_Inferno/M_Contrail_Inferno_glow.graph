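# Translucent unlit two-sided glow material for the Inferno skydiving contrail, used on Niagara sprite and mesh particles, colored by particle color.
# A radial sphere mask with dynamic-parameter power is multiplied by particle alpha, depth and near-camera fades (skipped on forward and mobile paths), and optional view-angle fade.
# An inverted sphere mask cuts a dynamically sized "black hole" in the center.
# A scalar can blend this to a simple center mask.
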Material /BRCosmetics/Athena/Effects/Particles/Skydiving/NS_Contrail_Inferno/M_Contrail_Inferno_glow
graph {
 properties {
  bAutomaticallySetUsageInEditor=false
  bUsedWithNiagaraMeshParticles=true
  bUsedWithNiagaraSprites=true
  blendMode=BLEND_Translucent
  pixelDepthOffsetMode=PDOM_Legacy
  shadingModel=MSM_Unlit
  twoSided=true
  usedShadingModels=Unlit
 }
 nodes [
  n1 = TextureCoordinate()
  n2 = Constant() {r=0.5}
  n3 = ScalarParameter() {defaultValue=0.5, parameterName=Radius}
  n4 = ScalarParameter() {parameterName=Hardness}
  n5 = DynamicParameter() {defaultValue=<rgba 3 1 1 1>, paramNames [Power, DepthFadeDistMult, BlackHole size, Param4]}
  n6 = ParticleColor()
  n7 = DepthFade(
   Opacity=MaterialFunctionCall(Distance=ScalarParameter() {defaultValue=100, parameterName=NearCameraFadeDist}) {materialFunction=&Near_Camera_Fade}.Result
   FadeDistance=Multiply(A=@n5.DepthFadeDistMult, B=ScalarParameter() {defaultValue=100, parameterName=DepthFadeDist})
  )
  n8 = Constant() {r=1}
  n9 = Multiply(
   A=Power(
    Base=SphereMask(A=@n1, B=@n2, Radius=@n3, Hardness=@n4) {hardnessPercent=0}
    Exp=Clamp(None=@n5.Power) {clampMode=CMODE_ClampMin, maxDefault=100, minDefault=1}
   ) {constExponent=3}
   B=Multiply(A=@n6.A, B=ShadingPathSwitch(Default=@n7, Deferred=@n7, Forward=@n8, Mobile=@n8))
  )
  n10 = MaterialFunctionCall("FadeSensitivity (optional)"=ScalarParameter() {defaultValue=1.2, parameterName=AngleFade}) {materialFunction=&FadeBasedOnViewAngle}
  n11 = LinearInterpolate(
   A=Constant() {r=1}
   B=Saturate(None=Multiply(A=@n10.Result) {constB=10})
   Alpha=ScalarParameter() {defaultValue=1, parameterName=Use AngleFade}
  )
  n12 = ScalarParameter() {defaultValue=0.5, parameterName=BlackHole}
  n13 = DynamicParameter() {defaultValue=<rgba 3 1 1 1>, paramNames [Power, DepthFadeDistMult, BlackHole size, Param4]}
  n14 = OneMinus(None=SphereMask(A=TextureCoordinate(), B=Constant() {r=0.5}, Radius=Multiply(A=@n12, B=@n13.BlackHole size)) {attenuationRadius=0.5, hardnessPercent=0})
  n15 = SphereMask(A=TextureCoordinate(), B=Constant() {r=0.5}, Radius=Constant() {r=0.25}, Hardness=Constant() {r=0.7})
  LinearInterpolate(
   A=Multiply(A=Multiply(A=@n9, B=@n11), B=Power(Base=@n14))
   B=Multiply(A=ParticleColor().A, B=Power(Base=@n15) {constExponent=4})
   Alpha=ScalarParameter() {parameterName=Use simple center mask}
  )
  MaterialFunctionCall(Emissive=Multiply(A=@n6.RGB)) {materialFunction=&MF_FX_EmissiveAdjustments}
 ]
}
